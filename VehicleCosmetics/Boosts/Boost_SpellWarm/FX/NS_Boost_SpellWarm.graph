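# Niagara system for the SpellWarm vehicle boost cosmetic: a warm-colored exhaust at a single boost position plus the car offset.
# Cone emitter layers four SpawnGroups, switching size, offset and velocity per group, with VisibilityTag, curl noise and painted-variant color overrides.
# Embers emitter shoots short-lived, dragged sparks backward with a separately sized glow.
# Ribbon emitter draws a trail with animated material parameter, changing lifetime and speed for frontend preview.
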NiagaraSystem /VehicleCosmetics/Boosts/Boost_SpellWarm/FX/NS_Boost_SpellWarm
dependencies {
 usedDataInterfaces [
  &NiagaraDataInterfaceArrayFloat3
  &NiagaraDataInterfaceCurve
  &NiagaraDataInterfaceVector2DCurve
  &NiagaraDataInterfaceColorCurve
  &NiagaraDataInterfaceSkeletalMesh
 ]
 usedDynamicInputs []
 usedModules [
  &self:SystemSpawnScript.SystemScriptSource.SystemScriptGraph.NiagaraNodeAssignment_0.SetVariables_#1
  &SystemState
 ]
}
emitters {
 NE_VC_SingleBoost_Cone {
  emitterUpdateScript [
   &EmitterState()
   &SpawnRate(
    SpawnRate=&RandomRangeFloat(Minimum=50, Maximum=60, Recalculate Random Each Loop=false)
    Use Spawn Probability=false
   )
   &SpawnRate(
    SpawnGroup=1
    SpawnRate=&RandomRangeFloat(Minimum=60, Maximum=80, Recalculate Random Each Loop=false)
    Use Spawn Probability=false
   ) {module=SpawnRate001}
   &SpawnRate(
    SpawnGroup=2
    SpawnRate=&RandomRangeFloat(Minimum=60, Maximum=70, Recalculate Random Each Loop=false)
    Use Spawn Probability=false
   ) {module=SpawnRate002}
   &SpawnRate(
    SpawnGroup=3
    SpawnRate=&RandomRangeFloat(Minimum=70, Maximum=80, Recalculate Random Each Loop=false)
    Use Spawn Probability=false
   ) {module=SpawnRate003}
  ]
  particleSpawnScript [
   &VC_PaintedVariantsOverride(
    DefaultColor=<rgba 0 0 0 1>
    Orange=<rgba 0.6724 0.2346 0 1>
    R1 DefaultColor=&ScaleLinearColorByCurve(
     InitialValue=<rgba 1 0.13 0.0139 1>
     ScaleCurve=DataInterface() {propertyValues {
       AlphaCurve {
        defaultValue=1
        keys [
         {tangentMode=RCTM_User, time=0, value=1}
         {arriveTangent=-0.0405, interpMode=RCIM_Cubic, leaveTangent=-0.0405, tangentMode=RCTM_User, time=0.5308, value=1}
         {arriveTangent=-8.0089, interpMode=RCIM_Cubic, leaveTangent=-8.0089, tangentMode=RCTM_User, time=1, value=0}
        ]
       }
       BlueCurve {keys [<at 0.0796 1>]}
       GreenCurve {keys [<at 0.0796 1>]}
       RedCurve {keys [<at 0.0796 1>]}
      }}
     CurveIndex=$Particles.NormalizedAge
    )
    R2 DefaultColor:LinearColor=<rgba 1 0.9081 0.1035 1>
    R3 DefaultColor:LinearColor=<rgba 0.5119 0.5119 0.5119 1>
    Use R1 Painted Overrides=true
    Use R2 Painted Overrides=true
    Use R3 Painted Overrides=true
   ) {enabled=false}
   /SetVariables(Particles.VisibilityTag=$Emitter.SpawnGroup) {isSetParameters=true}
   &InitializeParticle(
    Color Mode=::NewEnumerator0
    Lifetime Max=&MergeHelper_FloatSwitch(0=0.8, 1=0.7, 2=0.9, 3=1, Index=$Emitter.SpawnGroup, Num=::NewEnumerator2)
    Lifetime Min:NiagaraFloat=&MergeHelper_FloatSwitch(0=0.5, 1=0.5, 2=0.5, 3=0.5, Index=$Emitter.SpawnGroup, Num=::NewEnumerator2)
    Lifetime Mode=::NewEnumerator1
    Position Offset=&MergeHelper_VectorSwitch(
     0=&AddVectorToPosition(
      Position=<-61 0 0>
      Vector=&AddVectorToPosition(Position=$User.DTCarOffset, Vector=Linked() {linkedVariable {name=User.SingleBoostPos, type=NiagaraPosition}})
     )
     1=&AddVectorToPosition(
      Position=<-68 0 0>
      Vector=&AddVectorToPosition(Position=$User.DTCarOffset, Vector=Linked() {linkedVariable {name=User.SingleBoostPos, type=NiagaraPosition}})
     )
     2=&AddVectorToPosition(
      Position=<-68 0 0>
      Vector=&AddVectorToPosition(Position=$User.DTCarOffset, Vector=Linked() {linkedVariable {name=User.SingleBoostPos, type=NiagaraPosition}})
     )
     3=&AddVectorToPosition(
      Position=&MakeVector(X=-76, Y=0, Z=&RandomRangeFloat(Minimum=3, Maximum=-3))
      Vector=&AddVectorToPosition(Position=$User.DTCarOffset, Vector=Linked() {linkedVariable {name=User.SingleBoostPos, type=NiagaraPosition}})
     )
     Index=$Emitter.SpawnGroup
     Num=::NewEnumerator2
    )
    Sprite Rotation Mode=::NewEnumerator2
    Sprite Size Mode=::NewEnumerator1
    Uniform Sprite Size Max=&MergeHelper_FloatSwitch(0=100, 1=70, 2=60, 3=25, Index=$Emitter.SpawnGroup, Num=::NewEnumerator2)
    Uniform Sprite Size Min=&MergeHelper_FloatSwitch(0=90, 1=65, 2=45, 3=20, Index=$Emitter.SpawnGroup, Num=::NewEnumerator2)
    UsePositionOffset=true
    Write Material Random=true
   )
   &AddVelocity(
    Velocity=&RandomRangeVector(Minimum=<-100 0 -12>, Maximum=<-120 0 12>)
    Velocity Speed Scale=&MergeHelper_FloatSwitch(0=1, 1=1.4, 2=1.1, 3=1.2, Index=$Emitter.SpawnGroup, Num=::NewEnumerator2)
   ) {enabled=false}
   &AddVelocity(
    Rotation Coordinate Space=::Simulation
    Velocity=&MergeHelper_VectorSwitch(
     0=&RandomRangeVector(Minimum=<-150 0 -10>, Maximum=<-180 0 10>)
     1=&RandomRangeVector(
      Minimum=&SelectVectorFromArray(
       Array Sampling Mode=::NewEnumerator0
       Direct Array Index=?
       Vector Selection Array=DataInterface() {propertyValues {FloatData [<0 40 40>, <0 -40 -40>]}}
       Random Seed=0
      )
      Maximum=&RandomRangeVector(Minimum=<-300 15 0>, Maximum=<-320 -15 0>)
     )
     2=&RandomRangeVector(Minimum=<-140 15 -20>, Maximum=<-150 -15 20>)
     3=&RandomRangeVector(Minimum=<-140 15 -5>, Maximum=<-160 -15 10>)
     Index=$Emitter.SpawnGroup
     Num=::NewEnumerator2
    )
   ) {module=AddVelocity001}
  ]
  particleUpdateScript [
   &ParticleState()
   &AccelerationForce(Acceleration=<20 0 0>, Coordinate Space=::Local)
   &CurlNoiseForce(Bias Noise=false, Noise Frequency=100, Noise Strength=100, Pan Noise=false)
   &SpriteRotationRate(Rotation Rate=&RandomRangeFloat(Minimum=100, Maximum=250))
   &SolveForcesAndVelocity(Clamp Velocity=false, Limit Acceleration=false)
   &VC_PaintedVariantsOverride(
    DefaultColor=<rgba 0 0 0 1>
    R1 DefaultColor=&ScaleLinearColorByCurve(
     InitialValue=<rgba 0.722 0.0925 0.0092 1>
     ScaleCurve=DataInterface() {propertyValues {
       AlphaCurve {
        defaultValue=1
        keys [
         {tangentMode=RCTM_User, time=0, value=0.8}
         {arriveTangent=-8.0089, interpMode=RCIM_Cubic, leaveTangent=-8.0089, tangentMode=RCTM_User, time=0.8329, value=0}
        ]
       }
       BlueCurve {keys [<at 0 1>]}
       GreenCurve {keys [<at 0 1>]}
       RedCurve {keys [<at 0 1>]}
      }}
     CurveIndex=$Particles.NormalizedAge
    )
    R2 DefaultColor:LinearColor=&ScaleLinearColorByCurve(
     InitialValue=<rgba 0.6802 0.3435 0.034 1>
     ScaleCurve=DataInterface() {propertyValues {
       AlphaCurve {
        defaultValue=1
        keys [
         {tangentMode=RCTM_User, time=0, value=1}
         {arriveTangent=-8.0089, interpMode=RCIM_Cubic, leaveTangent=-8.0089, tangentMode=RCTM_User, time=1, value=0}
        ]
       }
       BlueCurve {keys [<at 0 1>]}
       GreenCurve {keys [<at 0 1>]}
       RedCurve {keys [<at 0 1>]}
      }}
     CurveIndex=$Particles.NormalizedAge
    )
    R3 DefaultColor:LinearColor=&ScaleLinearColorByCurve(
     InitialValue=<rgba 0.5793 0.3901 0.3312 1>
     ScaleCurve=DataInterface() {propertyValues {
       AlphaCurve {
        defaultValue=1
        keys [
         {tangentMode=RCTM_User, time=0, value=1}
         {arriveTangent=-0.0405, interpMode=RCIM_Cubic, leaveTangent=-0.0405, tangentMode=RCTM_User, time=0.5308, value=1}
         {arriveTangent=-8.0089, interpMode=RCIM_Cubic, leaveTangent=-8.0089, tangentMode=RCTM_User, time=0.9523, value=0}
        ]
       }
       BlueCurve {keys [<at 0 0.7>]}
       GreenCurve {keys [<at 0 0.7>]}
       RedCurve {keys [<at 0 1>]}
      }}
     CurveIndex=$Particles.NormalizedAge
    )
    Use R1 Painted Overrides=true
    Use R2 Painted Overrides=true
    Use R3 Painted Overrides=true
   ) {module=VC_PaintedVariantsOverride001}
   &ScaleSpriteSize(Uniform Curve Sprite Scale=DataInterface() {propertyValues {Curve {
       defaultValue=0
       keys [
        {tangentMode=RCTM_User, time=0, value=1}
        {arriveTangent=-0.0471, interpMode=RCIM_Cubic, leaveTangent=-0.0471, tangentMode=RCTM_User, time=0.5, value=1}
        {arriveTangent=-3.4858, interpMode=RCIM_Cubic, leaveTangent=-3.4858, tangentMode=RCTM_User, time=1, value=0.3419}
       ]
      }}})
   &ScaleColor(
    Scale Alpha=&FloatFromCurve(FloatCurve=DataInterface() {propertyValues {Curve {
        defaultValue=1
        keys [
         {tangentMode=RCTM_User, time=0, value=1}
         {arriveTangent=-0.4455, interpMode=RCIM_Cubic, leaveTangent=-0.4455, tangentMode=RCTM_User, time=0.45, value=1}
         {arriveTangent=-4.391, interpMode=RCIM_Cubic, leaveTangent=-4.391, tangentMode=RCTM_User, time=1, value=0}
        ]
       }}})
    Scale RGB=<0 0 0>
    ScaleA=true
    ScaleRGB=true
   )
   /SetVariables(Particles.ORG_SIZE:Vector2f=&Multiply_Vector2D_ByFloat(
     Float=&RandomRangeFloat(Minimum=65, Maximum=70)
     Vector2D=&Vector2DFromCurve(
      Vector2Curve=DataInterface() {propertyValues {XCurve {keys [<at 0.4 1 RCIM_Cubic>,<at 1 0.4 RCIM_Cubic>]}, YCurve {keys [<at 0.4 1 RCIM_Cubic>,<at 1 0.4 RCIM_Cubic>]}}}
      ScaleCurve=<1 1>
     )
    )) {isSetParameters=true}
  ]
  rendererClasses [&NiagaraSpriteRendererProperties]
  renderers [
   {rendererClass=&NiagaraSpriteRendererProperties}
   {rendererClass=&NiagaraSpriteRendererProperties, rendererIndex=1}
   {rendererClass=&NiagaraSpriteRendererProperties, rendererIndex=2}
   {rendererClass=&NiagaraSpriteRendererProperties, rendererIndex=3}
  ]
 }
 NE_VC_SingleBoost_Embers {
  emitterUpdateScript [
   &EmitterState()
   &SpawnRate(SpawnRate=&RandomRangeFloat(Minimum=50, Maximum=60), Use Spawn Probability=false)
  ]
  particleSpawnScript [
   &VC_PaintedVariantsOverride(DefaultColor=<rgba 0.9911 1 0.0307 1>, R1 DefaultColor=<rgba 10 0 0 1>, Use R1 Painted Overrides=true)
   &InitializeParticle(
    Color Mode=::NewEnumerator0
    Lifetime Max=0.4
    Lifetime Min=0.35
    Lifetime Mode=::NewEnumerator1
    Position Offset=&AddVectorToPosition(
     Position=<0 0 0>
     Vector=&AddVectorToPosition(Position=$User.DTCarOffset, Vector=Linked() {linkedVariable {name=User.SingleBoostPos, type=NiagaraPosition}})
    )
    Position Offset Coordinate Space=::Local
    Sprite Rotation Mode=::NewEnumerator2
    Sprite Size Max=<3 35>
    Sprite Size Min=<2 10>
    Sprite Size Mode=::NewEnumerator2
    UsePositionOffset=true
    Write Material Random=true
   )
   &AddVelocity(Velocity=&SelectVectorFromArray(
     Array Sampling Mode=::NewEnumerator0
     Direct Array Index=?
     Vector Selection Array=DataInterface() {propertyValues {FloatData [<-1650 0 42>, <-1650 0 -40>]}}
     Random Seed=0
    ))
   &ShapeLocation(
    Handle Radius=10
    Large Radius=15
    MaskSpawnGroup=false
    Override Seed=false
    Rotation Mode=::NewEnumerator1
    Shape Primitive=::NewEnumerator3
    Surface Distribution=1
    UseOwnerScale=true
   )
  ]
  particleUpdateScript [
   &ParticleState()
   &Drag(
    Drag=&FloatFromCurve(
     FloatCurve=DataInterface() {propertyValues {Curve {
        defaultValue=1
        keys [<at 0.2 0 RCIM_Cubic>, {arriveTangent=0.972, interpMode=RCIM_Cubic, leaveTangent=0.972, tangentMode=RCTM_User, time=1, value=1}]
       }}}
     Scale Curve=2
    )
    Use Linear Drag=true
    Use Rotational Drag=true
   )
   &SolveForcesAndVelocity(Clamp Velocity=false, Limit Acceleration=false)
   &ScaleSpriteSize(
    Non-Uniform Curve Scale=<2 2>
    Non-Uniform Curve Sprite Scale=DataInterface() {propertyValues {
      XCurve {keys [<at 0 0.1>, {tangentMode=RCTM_User, time=0.5, value=0.6}, <at 1 0>]}
      YCurve {keys [{tangentMode=RCTM_User, time=0, value=0.1}, {tangentMode=RCTM_User, time=0.5, value=1}, <at 1 0>]}
     }}
    Scale Sprite Size Mode=::NewEnumerator3
   )
   /SetVariables(Particles.SizeGlow:Vector2f=&Multiply_Vector2D_ByFloat(Float=1.3, Vector2D=$Particles.SpriteSize)) {isSetParameters=true}
  ]
  rendererClasses [&NiagaraSpriteRendererProperties]
  renderers [
   {rendererClass=&NiagaraSpriteRendererProperties}
   {rendererClass=&NiagaraSpriteRendererProperties, rendererIndex=1}
  ]
 }
 NE_VC_SingleBoost_Ribbon {
  emitterUpdateScript [
   &EmitterState(Inactive Response=::NewEnumerator1, Life Cycle Mode=::NewEnumerator1, Loop Duration=1, UseLoopDelay=false)
   &VC_FilteredSocketCount()
   &VC_RateQualityHandler(Epic Spawn Rate=25, High Spawn Rate=25, Low Spawn Rate=25, Medium Spawn Rate=25, Use Custom Rate?=true)
   &VC_SpawnRibbon_v1(SpawnRate=$Emitter.Spawn Rate)
  ]
  particleSpawnScript [
   &VC_PaintedVariantsOverride() {enabled=false}
   &VC_InitializeRibbon_v1(
    Custom Offset=true
    Lifetime=&MakeCustomFloatFromBool(Bool=$User.bIsFrontendPreview, True Float=0.9, False Float=0.5)
    Local Vector Offset=&AddVectorToPosition(
     Position=<-50 -8 -2.5>
     Vector=&AddVectorToPosition(Position=$User.DTCarOffset, Vector=Linked() {linkedVariable {name=User.SingleBoostPos, type=NiagaraPosition}})
    )
    Offset Amount=1
    Ribbon Width=103
   )
   &AddVelocity(Velocity=&MakeCustomVectorFromBool(Bool=$User.bIsFrontendPreview, True Vector=<-1600 0 0>, False Vector=<-3000 0 0>)) {module=AddVelocity001}
  ]
  particleUpdateScript [
   &ParticleState()
   &InheritSourceMovement(
    Applied Position Scale=&Multiply_VectorByFloat(Vector=<1 1 1>, Float=1)
    ApplyForce=false
    ApplyPosition=true
    ApplyPositionFalloff=false
   )
   &SolveForcesAndVelocity(Clamp Velocity=false, Limit Acceleration=false)
   &VC_UpdateRibbon_v1()
   &ScaleColor(
    Scale RGB=&VectorFromFloat(Value=&FloatFromCurve(FloatCurve=DataInterface() {propertyValues {Curve {
         defaultValue=0
         keys [
          {arriveTangent=4.1785, interpMode=RCIM_Cubic, leaveTangent=4.1785, tangentMode=RCTM_User, time=0, value=1}
          {arriveTangent=0.0224, interpMode=RCIM_Cubic, leaveTangent=0.0224, tangentMode=RCTM_User, time=0.3436, value=1}
          {arriveTangent=-1.7973, interpMode=RCIM_Cubic, leaveTangent=-1.7973, time=0.65, value=1}
          {arriveTangent=-0.2547, interpMode=RCIM_Cubic, leaveTangent=-0.2547, tangentMode=RCTM_User, time=0.9, value=0}
         ]
        }}}))
    ScaleA=true
    ScaleRGB=true
   )
   &ScaleRibbonWidth(Ribbon Width Scale=&FloatFromCurve(FloatCurve=DataInterface() {propertyValues {Curve {keys [
         {arriveTangent=0.669, interpMode=RCIM_Cubic, leaveTangent=0.669, tangentMode=RCTM_User, time=0, value=0.9195}
         {arriveTangent=0.0123, interpMode=RCIM_Cubic, leaveTangent=0.0123, tangentMode=RCTM_User, time=0.0923, value=1}
         {interpMode=RCIM_Cubic, tangentMode=RCTM_User, time=0.55, value=1}
         <at 1 1.1>
        ]}}}))
   &DynamicMaterialParameters(
    Index 1 Param 1=&FloatFromCurve(FloatCurve=DataInterface() {propertyValues {Curve {keys [
         {arriveTangent=-13.683, interpMode=RCIM_Cubic, leaveTangent=-13.683, tangentMode=RCTM_User, time=0, value=1}
         {arriveTangent=-4.0398, time=0.25, value=-1.6429}
         {arriveTangent=7.5239, interpMode=RCIM_Cubic, leaveTangent=7.5239, time=0.7, value=-0.7}
         {arriveTangent=-0.0339, interpMode=RCIM_Cubic, leaveTangent=-0.0339, tangentMode=RCTM_User, time=1, value=4}
        ]}}})
    Module.Index0Float0Write=false
    Module.Index0Float1Write=false
    Module.Index0Float2Write=false
    Module.Index0Float3Write=false
    Module.Index1Float0Write=true
    Module.Index1Float1Write=false
    Module.Index1Float2Write=false
    Module.Index1Float3Write=false
    Write Parameter Index 1=::NewEnumerator1
   )
  ]
  rendererClasses [&NiagaraRibbonRendererProperties]
  renderers [{rendererClass=&NiagaraRibbonRendererProperties}]
 }
}
summary {
 systemName=NS_Boost_SpellWarm
 emitters [
  {emitterName=NE_VC_SingleBoost_Embers, rendererClasses [&NiagaraSpriteRendererProperties]}
  {emitterName=NE_VC_SingleBoost_Cone, rendererClasses [&NiagaraSpriteRendererProperties]}
  {emitterName=NE_VC_SingleBoost_Ribbon, rendererClasses [&NiagaraRibbonRendererProperties]}
 ]
 userVariables [
  User.bIsFrontendPreview {type=NiagaraBool}
  User.bIsPainted {type=NiagaraBool}
  User.ColorVariantID {type=NiagaraInt32}
  User.DTCarOffset {type=NiagaraPosition, description="", defaultValue {}}
  User.GreyTColor {type=LinearColor, description="", defaultValue {struct=LinearColor, value=<rgba 1 1 1 1>}}
  User.MTrailColor {type=LinearColor, description="", defaultValue {struct=LinearColor, value=<rgba 1 1 1 1>}}
  User.PaintedColor {type=LinearColor, description="", defaultValue {struct=LinearColor, value=<rgba 1 1 1 1>}}
  User.PaintedScalar {type=NiagaraFloat, defaultValue=1, description=""}
  User.PortCount {type=NiagaraInt32, defaultValue=1}
  User.SingleBoostPos {type=NiagaraPosition, description="", defaultValue {}}
  User.SndTColor {type=LinearColor, description="", defaultValue {struct=LinearColor, value=<rgba 1 1 1 1>}}
 ]
}
